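# Physics asset for the head part of the cinematic Banana character cosmetic, covering its right-side flap bone chain.
# Three box bodies form a chain from spine_05 through a tiny dyn_flapBase_r to the larger dyn_flap_r.
# The base joint rotates freely on all axes against the spine, while the flap joint has asymmetric angular limits, widest on swing 1 and very tight on twist.
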PhysicsAsset /BRCosmetics/Characters/Player/Male/Medium/Bodies/M_MED_Banana/Meshes/Cine/Cine_Parts/M_MED_Banana0202Cine_Head_Physics
bodies {
 dyn_flapBase_r [{shapeType=Box, extentX=1.01, extentY=1.01, extentZ=1.01, rotation=<rot 0 -90 0>}]
 dyn_flap_r [{shapeType=Box, extentX=5.0519, extentY=12.1485, extentZ=17.3234, center=<8.6465 -0.0667 0.7358>, rotation=<rot -11.0289 -93.7894 86.7641>}]
 spine_05 [{shapeType=Box, extentX=12.5928, extentY=28.8354, extentZ=16.1629}]
}
constraints [
 {
  bone1Name=dyn_flapBase_r
  bone2Name=spine_05
  swing1LimitDegrees=45
  swing1Motion=Free
  swing2LimitDegrees=45
  swing2Motion=Free
  twistLimitDegrees=45
  twistMotion=Free
 }
 {bone1Name=dyn_flap_r, bone2Name=dyn_flapBase_r, swing1LimitDegrees=45, swing2LimitDegrees=10, twistLimitDegrees=3}
]
