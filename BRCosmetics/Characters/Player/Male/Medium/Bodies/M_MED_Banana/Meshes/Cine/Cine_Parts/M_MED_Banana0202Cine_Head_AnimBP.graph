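# Animation Blueprint for the Banana skin's cinematic head part, copying pose and curves from its attached parent mesh and adding secondary motion to the peel.
# Constraints tie the right flap base and big peel bones to spine_05 and neck_01, feeding a RigidBody simulation in root-bone space.
# A final bone transform overrides the flap base rotation.
# A second InnerFlap RigidBody and a facial idle player are unconnected, and the event graph has no logic.
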AnimBlueprint /BRCosmetics/Characters/Player/Male/Medium/Bodies/M_MED_Banana/Meshes/Cine/Cine_Parts/M_MED_Banana0202Cine_Head_AnimBP
target_skeleton=&/M_MED_Banana0202Cine_Head_Skeleton
graphs [
 AnimGraph {nodes [
   RigidBody() {
    bClampLinearTranslationLimitToRefPose=true
    bOverrideWorldGravity=true
    overridePhysicsAsset=/Game/Characters/Player/Male/Medium/Bodies/M_MED_Banana/Meshes/M_MED_Banana_InnerFlap_Physics
    simulationSpace=BaseBoneSpace
    alphaBoolBlend {blendOption=Linear}
    alphaScaleBiasClamp {bInterpResult=true, interpSpeedDecreasing=2, interpSpeedIncreasing=2}
    baseBoneRef {boneName=root}
    componentAppliedLinearAccClamp=<800 800 800>
    componentLinearAccScale=<0.65 0.65 0.65>
    componentLinearVelScale=<0.3 0.3 0.3>
    overrideWorldGravity {z=-980}
   }
   n1 = Copy Pose From Mesh() {bCopyCurves=true, bUseAttachedParent=true}
   n2 = "Constraint - dyn_flapBase_r (PP)"!(
    ComponentPose=Local To Component(LocalPose=Layered blend per bone(BasePose=@n1))
    Alpha=1
    ConstraintWeights_0=0.8
    ConstraintWeights_1:float=0.2
   ) -> (Pose) {
    alphaBoolBlend {blendOption=Linear}
    boneToModify {boneName=dyn_flapBase_r}
    constraintSetup [{targetBone {boneName=spine_05}}, {targetBone {boneName=neck_01}}]
    constraintWeights [1,1]
   }
   n3 = "Constraint - dyn_bigPeel_01 (PP)"!(ComponentPose=@n2, Alpha=1, ConstraintWeights_0=0.8, ConstraintWeights_1:float=0.2) -> (Pose) {
    alphaBoolBlend {blendOption=Linear}
    boneToModify {boneName=dyn_bigPeel_01}
    constraintSetup [{targetBone {boneName=spine_05}}, {targetBone {boneName=neck_01}}]
    constraintWeights [1,1]
   }
   n4 = RigidBody(ComponentPose=@n3) {
    bClampLinearTranslationLimitToRefPose=true
    bOverrideWorldGravity=true
    overridePhysicsAsset=&/M_MED_Banana0202Cine_Head_Physics
    simulationSpace=BaseBoneSpace
    alphaBoolBlend {blendOption=Linear}
    baseBoneRef {boneName=root}
    componentAppliedLinearAccClamp=<800 800 800>
    componentLinearAccScale=<0.65 0.65 0.65>
    componentLinearVelScale=<0.3 0.3 0.3>
    overrideWorldGravity {z=-980}
   }
   n5 = "Transform (Modify) Bone - Bone: dyn_flapBase_r"(ComponentPose=@n4, Alpha=1, Translation="0,0,0", Rotation="0,-20,0", Scale="1,1,1") {
    rotationMode=BMM_Replace
    rotationSpace=BCS_BoneSpace
    alphaBoolBlend {blendOption=Linear}
    alphaScaleBiasClamp {bInterpResult=true, interpSpeedDecreasing=3.5, interpSpeedIncreasing=8}
    boneToModify {boneName=dyn_flapBase_r}
    rotation {yaw=-20}
   }
   Output Pose(Result=Component To Local(ComponentPose=@n5)) {layerGroup=None}
   Sequence Player 'Banana_Male_Facial_Idle'() {sequence=/Game/Animation/Game/MainPlayer/Facial/Banana/Banana_Male_Facial_Idle}
  ]}
 EventGraph {nodes [
   TryGetPawnOwner()
   Event BlueprintUpdateAnimation()
  ]}
]
variables []
